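# Sound cue for collecting a battery ingredient, classed as a player interaction sound.
# A mixer layers a "Real" wave at full volume with a "Surreal" wave at half volume.
# It uses the prop collect attenuation and overrides concurrency to allow two instances, stopping the oldest when exceeded.
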SoundCue /Game/Sounds/Creative/Ingredients/CP_Battery_Collect_Cue
root=Mixer_8
nodes [
 {id=Mixer_8, children [WavePlayer_18, WavePlayer_17], properties {inputVolume [1,0.5]}}
 {id=WavePlayer_17, properties {soundWaveAssetPtr=&/CP_Battery_Collect_Surreal}}
 {id=WavePlayer_18, properties {soundWaveAssetPtr=&/CP_Battery_Collect_Real}}
]
properties {
 attenuationSettings=/Game/Sounds/Attenuation/Fort_Prop_Collect_Att
 bOverrideConcurrency=true
 maxDistance=2097152
 soundClassObject=/Game/Sounds/FortSoundClassesAndModes/Fort_08_Player_Interacts
 concurrencyOverrides {maxCount=2, resolutionRule=StopOldest}
}
