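# Sound cue for collecting the Creative copper ingredient, played through the loot-collect sound class with prop-collect attenuation.
# A Mixer layers a slightly lowered bag foley wave under the copper collect wave, and a fixed Delay at the root holds back both.
# The cue runs at a low overall volume and overrides concurrency to allow two instances, stopping the oldest.
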SoundCue /Game/Sounds/Creative/Ingredients/CP_Copper_Collect_Cue
root=Delay_1
nodes [
 {id=Delay_1, children [Mixer_0], properties {delayMax=1.4, delayMin=1.4}}
 {id=Mixer_0, children [WavePlayer_5,WavePlayer_0], properties {inputVolume [0.75,1]}}
 {id=WavePlayer_0, properties {soundWaveAssetPtr=&/CP_Copper_Collect}}
 {id=WavePlayer_5, properties {soundWaveAssetPtr=/Game/Sounds/Foley_Loot/Ingredients/Bag/loot_setup_bagonly_01}}
]
properties {
 attenuationSettings=/Game/Sounds/Attenuation/Fort_Prop_Collect_Att
 bOverrideConcurrency=true
 maxDistance=2097152
 soundClassObject=&Fort_Loot_Collect
 volumeMultiplier=0.15
 concurrencyOverrides {maxCount=2, resolutionRule=StopOldest}
}
