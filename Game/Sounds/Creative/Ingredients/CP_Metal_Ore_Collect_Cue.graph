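# Sound cue for collecting the metal ore ingredient, routed to the loot collect sound class with prop collect attenuation.
# A fixed delay precedes a mixer that layers a random pick of three ore collect variants over quieter surreal rare-metal and loot bag foley layers.
# A low volume multiplier scales the result, and a concurrency override allows two instances at once, stopping the oldest.
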SoundCue /Game/Sounds/Creative/Ingredients/CP_Metal_Ore_Collect_Cue
root=Delay_1
nodes [
 {id=Delay_1, children [Mixer_5], properties {delayMax=1.4, delayMin=1.4}}
 {id=Mixer_5, children [Random_8, WavePlayer_35, WavePlayer_36], properties {inputVolume [1.3,0.5,0.6]}}
 {id=Random_8, children [WavePlayer_37, WavePlayer_38, WavePlayer_39], properties {weights [1,1,1]}}
 {id=WavePlayer_35, properties {soundWaveAssetPtr=&/CP_Metal_Rare_Collect_Surreal}}
 {id=WavePlayer_36, properties {soundWaveAssetPtr=/Game/Sounds/Foley_Loot/Ingredients/Bag/loot_setup_bagonly_01}}
 {id=WavePlayer_37, properties {soundWaveAssetPtr=&/CP_ore_new_collect_03}}
 {id=WavePlayer_38, properties {soundWaveAssetPtr=&/CP_ore_new_collect_02}}
 {id=WavePlayer_39, properties {soundWaveAssetPtr=&/CP_ore_new_collect_01}}
]
properties {
 attenuationSettings=/Game/Sounds/Attenuation/Fort_Prop_Collect_Att
 bOverrideConcurrency=true
 maxDistance=2097152
 soundClassObject=&Fort_Loot_Collect
 volumeMultiplier=0.2
 concurrencyOverrides {maxCount=2, resolutionRule=StopOldest}
}
