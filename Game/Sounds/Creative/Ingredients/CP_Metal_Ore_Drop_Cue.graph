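# Sound cue for dropping the metal ore ingredient in Creative, classed as loot collect audio.
# A fixed delay precedes a mixer that layers a randomly chosen ore collect variant, emphasized over a surreal rare-metal collect sound and a bag foley layer.
# It uses weapon-drop attenuation, and overridden concurrency caps it at two instances, stopping the oldest.
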SoundCue /Game/Sounds/Creative/Ingredients/CP_Metal_Ore_Drop_Cue
root=Delay_0
nodes [
 {id=Delay_0, children [Mixer_0], properties {delayMax=1.4, delayMin=1.4}}
 {id=Mixer_0, children [Random_1,WavePlayer_6,WavePlayer_7], properties {inputVolume [1.3,0.5,0.6]}}
 {id=Random_1, children [WavePlayer_3,WavePlayer_4,WavePlayer_5], properties {weights [1,1,1]}}
 {id=WavePlayer_3, properties {soundWaveAssetPtr=&/CP_ore_new_collect_03}}
 {id=WavePlayer_4, properties {soundWaveAssetPtr=&/CP_ore_new_collect_02}}
 {id=WavePlayer_5, properties {soundWaveAssetPtr=&/CP_ore_new_collect_01}}
 {id=WavePlayer_6, properties {soundWaveAssetPtr=&/CP_Metal_Rare_Collect_Surreal}}
 {id=WavePlayer_7, properties {soundWaveAssetPtr=/Game/Sounds/Foley_Loot/Ingredients/Bag/loot_setup_bagonly_01}}
]
properties {
 attenuationSettings=&Fort_Weapon_Drop_Att
 bOverrideConcurrency=true
 maxDistance=3000
 soundClassObject=&Fort_Loot_Collect
 volumeMultiplier=0.159
 concurrencyOverrides {maxCount=2, resolutionRule=StopOldest}
}
