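# Sound cue for collecting a mushroom ingredient in Creative.
# A Mixer layers two waves at equal volume, a "Real" and a "Surreal" collect sound, so both play together.
# The cue uses the loot-collect sound class, the shared prop-collect attenuation and a reduced overall volume.
# A concurrency override caps it at two simultaneous instances, stopping the oldest when exceeded.
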SoundCue /Game/Sounds/Creative/Ingredients/CP_Mushroom_Collect_Cue
root=Mixer_10
nodes [
 {id=Mixer_10, children [WavePlayer_22, WavePlayer_7], properties {inputVolume [1,1]}}
 {id=WavePlayer_22, properties {soundWaveAssetPtr=&/CP_Mushroom_Collect_Real}}
 {id=WavePlayer_7, properties {soundWaveAssetPtr=&/CP_Mushroom_Collect_Surreal02}}
]
properties {
 attenuationSettings=/Game/Sounds/Attenuation/Fort_Prop_Collect_Att
 bOverrideConcurrency=true
 maxDistance=2097152
 soundClassObject=&Fort_Loot_Collect
 volumeMultiplier=0.376
 concurrencyOverrides {maxCount=2, resolutionRule=StopOldest}
}
