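# Sound cue for collecting the powder crafting ingredient in Creative, played as a spatialized loot-collect sound with prop-collect attenuation.
# A mixer layers one of three equally weighted random ore-collect variations, a surreal powder sweetener and a bag foley setup sound at separate volumes.
# A fixed delay of over a second holds back the whole mix, and a concurrency override caps it at two instances, stopping the oldest.
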SoundCue /Game/Sounds/Creative/Ingredients/CP_Powder_Collect_Cue
root=Delay_0
nodes [
 {id=Delay_0, children [Mixer_0], properties {delayMax=1.4, delayMin=1.4}}
 {id=Mixer_0, children [Random_0, WavePlayer_65, WavePlayer_1], properties {inputVolume [1.3,0.8,0.6]}}
 {id=Random_0, children [WavePlayer_2,WavePlayer_3,WavePlayer_4], properties {weights [1,1,1]}}
 {id=WavePlayer_1, properties {soundWaveAssetPtr=/Game/Sounds/Foley_Loot/Ingredients/Bag/loot_setup_bagonly_01}}
 {id=WavePlayer_2, properties {soundWaveAssetPtr=&/CP_ore_new_collect_03}}
 {id=WavePlayer_3, properties {soundWaveAssetPtr=&/CP_ore_new_collect_02}}
 {id=WavePlayer_4, properties {soundWaveAssetPtr=&/CP_ore_new_collect_01}}
 {id=WavePlayer_65, properties {soundWaveAssetPtr=&/CP_Powder_Collect_Surreal}}
]
properties {
 attenuationSettings=/Game/Sounds/Attenuation/Fort_Prop_Collect_Att
 bOverrideConcurrency=true
 maxDistance=2097152
 soundClassObject=&Fort_Loot_Collect
 volumeMultiplier=0.351
 concurrencyOverrides {maxCount=2, resolutionRule=StopOldest}
}
