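# Sound cue for dropping the Powder ingredient, built from powder collect recordings and classed as loot collect audio.
# A mixer plays a realistic foley layer and a surreal layer together at equal input volume, with the whole cue turned down well below full level.
# It uses weapon drop attenuation for distance falloff and overrides concurrency to allow two instances, stopping the oldest.
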SoundCue /Game/Sounds/Creative/Ingredients/CP_Powder_Drop_Cue
root=Mixer_31
nodes [
 {id=Mixer_31, children [WavePlayer_64, WavePlayer_65], properties {inputVolume [1,1]}}
 {id=WavePlayer_64, properties {soundWaveAssetPtr=/Game/Sounds/Foley_Loot/Ingredients/Munitions/Powder_Collect_Real}}
 {id=WavePlayer_65, properties {soundWaveAssetPtr=&/CP_Powder_Collect_Surreal}}
]
properties {
 attenuationSettings=&Fort_Weapon_Drop_Att
 bOverrideConcurrency=true
 maxDistance=3000
 soundClassObject=&Fort_Loot_Collect
 volumeMultiplier=0.199
 concurrencyOverrides {maxCount=2, resolutionRule=StopOldest}
}
