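# Sound cue for collecting the Silver ingredient in Creative, played as a loot-collect sound.
# A mixer layers the Silver "Real" and "Surreal" waves with a shared bag-foley loot sample, with the latter two layers mixed lower.
# The cue plays at reduced volume through the prop-collect attenuation and the loot-collect sound class.
# A concurrency override allows two instances at once, stopping the oldest when exceeded.
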SoundCue /Game/Sounds/Creative/Ingredients/CP_Silver_Collect_Cue
root=Mixer_20
nodes [
 {id=Mixer_20, children [WavePlayer_40, WavePlayer_41, WavePlayer_10], properties {inputVolume [1,0.75,0.7]}}
 {id=WavePlayer_10, properties {soundWaveAssetPtr=/Game/Sounds/Foley_Loot/Ingredients/Bag/loot_setup_bagonly_01}}
 {id=WavePlayer_40, properties {soundWaveAssetPtr=&/CP_Silver_Collect_Real}}
 {id=WavePlayer_41, properties {soundWaveAssetPtr=&/CP_Silver_Collect_Surreal}}
]
properties {
 attenuationSettings=/Game/Sounds/Attenuation/Fort_Prop_Collect_Att
 bOverrideConcurrency=true
 maxDistance=2097152
 soundClassObject=&Fort_Loot_Collect
 volumeMultiplier=0.376
 concurrencyOverrides {maxCount=2, resolutionRule=StopOldest}
}
