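# Sound cue for collecting a sticky Creative ingredient, routed through the loot-collect sound class.
# A mixer plays two wave layers together at equal input volume, a Real and a Surreal variant, with the overall cue volume reduced.
# It uses the shared prop-collect attenuation and overrides concurrency to allow two instances, stopping the oldest.
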SoundCue /Game/Sounds/Creative/Ingredients/CP_Sticky_Collect_Cue
root=Mixer_3
nodes [
 {id=Mixer_3, children [WavePlayer_6,WavePlayer_8], properties {inputVolume [1,1]}}
 {id=WavePlayer_6, properties {soundWaveAssetPtr=&/CP_Sticky_Collect_Real}}
 {id=WavePlayer_8, properties {soundWaveAssetPtr=&/CP_Sticky_Collect_Surreal}}
]
properties {
 attenuationSettings=/Game/Sounds/Attenuation/Fort_Prop_Collect_Att
 bOverrideConcurrency=true
 maxDistance=2097152
 soundClassObject=&Fort_Loot_Collect
 volumeMultiplier=0.376
 concurrencyOverrides {maxCount=2, resolutionRule=StopOldest}
}
