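# Sound cue for collecting a wood plank crafting ingredient, routed through the loot collect sound class.
# A mixer layers one of three equally weighted random "Real" plank sounds with a constant "Surreal" layer.
# It uses prop collect attenuation and overrides concurrency to two simultaneous instances, stopping the oldest.
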SoundCue /Game/Sounds/Creative/Ingredients/CP_WoodPlank_Collect_Cue
root=Mixer_39
nodes [
 {id=Mixer_39, children [Random_0, WavePlayer_83], properties {inputVolume [1,1]}}
 {id=Random_0, children [WavePlayer_0,WavePlayer_1,WavePlayer_2], properties {weights [1,1,1]}}
 {id=WavePlayer_0, properties {soundWaveAssetPtr=&/CP_Woodplank_Collect_Real01}}
 {id=WavePlayer_1, properties {soundWaveAssetPtr=&/CP_Woodplank_Collect_Real02}}
 {id=WavePlayer_2, properties {soundWaveAssetPtr=&/CP_Woodplank_Collect_Real03}}
 {id=WavePlayer_83, properties {soundWaveAssetPtr=&/CP_WoodPlank_Collect_Surreal}}
]
properties {
 attenuationSettings=/Game/Sounds/Attenuation/Fort_Prop_Collect_Att
 bOverrideConcurrency=true
 maxDistance=2097152
 soundClassObject=&Fort_Loot_Collect
 volumeMultiplier=0.376
 concurrencyOverrides {maxCount=2, resolutionRule=StopOldest}
}
